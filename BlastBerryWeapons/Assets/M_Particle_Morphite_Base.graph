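# Base particle material for the Morphite look, built for Niagara sprite and mesh particles as a two-sided, Default Lit surface.
# It takes its full material attributes from the shared MF_Morphite_Base function.
# Only Emissive Color is changed, multiplied by the EmissiveMult scalar parameter to control glow strength, then written back into the attributes.
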Material /BlastBerryWeapons/Assets/M_Particle_Morphite_Base
graph {
 properties {
  bAutomaticallySetUsageInEditor=false
  bUseMaterialAttributes=true
  bUsedWithNiagaraMeshParticles=true
  bUsedWithNiagaraSprites=true
  twoSided=true
  usedShadingModels=Default Lit
 }
 nodes [
  n1 = GetMaterialAttributes(None=MaterialFunctionCall!(Override Normals, UVsOverride, BaseColorOverride, WorldPositionOverride, VertexNormalOverride) {materialFunction=/BlastBerryWeapons/Materials/Morphite/MF_Morphite_Base}.MA Out) {attributeGetTypes [#1]}
  SetMaterialAttributes(
   MaterialAttributes=@n1.MaterialAttributes
   Emissive Color=Multiply(A=@n1.Emissive Color, B=ScalarParameter() {parameterName=EmissiveMult}) {constB=0}
  ) {attributeSetTypes [#1]}
 ]
}
