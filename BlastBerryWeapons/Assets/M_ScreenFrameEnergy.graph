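# Additive unlit Niagara sprite material that draws a blue energy frame over the screen with depth testing disabled.
# Scrolling smoke distorts screen-aligned UVs fed into a powered sine edge mask, filled with panning electric noise times smoke.
# A WipeUpLerp parameter masks that layer vertically, and a VignetteRadius-driven inverted radial gradient adds an edge glow.
# Particle color and alpha scale the respective layers, which are tinted by a smoke-blended blue gradient before emissive adjustments.
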Material /BlastBerryWeapons/Assets/M_ScreenFrameEnergy
graph {
 properties {
  bAutomaticallySetUsageInEditor=false
  bDisableDepthTest=true
  bUsedWithNiagaraSprites=true
  blendMode=BLEND_Additive
  shadingModel=MSM_Unlit
  usedShadingModels=Unlit
 }
 nodes [
  n1 = Constant3Vector() {constant=<rgba 0 0.5392 1 1>}
  n2 = Constant3Vector() {constant=<rgba 0.1281 0.1885 1 1>}
  n3 = TextureCoordinate()
  n4 = TextureSample(UVs=Panner(Coordinate=Multiply(A=@n3) {constB=5}) {speedX=0.325, speedY=0.235}) {texture=/Game/Effects/Fort_Effects/Textures/Tile/T_Inky_Smoke_Tile}
  n5 = ParticleColor()
  n6 = MaterialFunctionCall() {materialFunction=&ScreenAlignedUVs}
  n7 = TextureSample(UVs=Panner(Coordinate=Multiply(A=@n6."X Scale to Ratio, Y 100%") {constB=8}) {speedX=0.35, speedY=0.6}) {texture=&T_Inky_Smoke_Tile}
  n8 = AppendVector(A=@n7.R, B=@n7.R)
  n9 = Multiply(A=Subtract(A=Multiply(A=@n8) {constB=2}) {constB=0.5}) {constB=0.04}
  n10 = Add(A=@n6."X 100%, Y 100%", B=@n9)
  n11 = Power(Base=Sine(None=ConstantBiasScale(None=@n10) {bias=-0.75, scale=1})) {constExponent=15}
  n12 = Add(A=ComponentMask(None=@n11) {r=true}, B=ComponentMask(None=@n11) {g=true})
  n13 = Add(A=@n9, B=Multiply(A=@n3) {constB=3})
  n14 = Multiply(
   A=TextureSample(UVs=Panner(Coordinate=@n13) {speedX=-0.1, speedY=-0.15}) {samplerType=SAMPLERTYPE_LinearColor, texture=&/T_ElectricNoise}.RGB
   B=@n4.RGB
  )
  n15 = TextureCoordinate()
  n16 = DynamicParameter() {defaultValue=<rgba 0.75 1 1 1>, paramNames [VignetteRadius, WipeUpLerp, Param3, Param4]}
  n17 = Subtract(A=OneMinus(None=ComponentMask(None=@n15) {g=true}), B=LinearInterpolate(Alpha=Saturate(None=@n16.WipeUpLerp)) {constA=-1})
  n18 = Multiply(A=Multiply(A=@n5.RGB, B=Multiply(A=@n12, B=@n14)), B=Saturate(None=OneMinus(None=@n17)))
  n19 = MaterialFunctionCall(UVs=@n6."X 100%, Y 100%", Radius=@n16.VignetteRadius, Density=Constant() {r=5}) {materialFunction=&RadialGradientExponential}
  n20 = Multiply(A=@n5.A, B=Multiply(A=OneMinus(None=@n19.RadialGradientExponential)) {constB=0.35})
  MaterialFunctionCall(Emissive=Multiply(A=LinearInterpolate(A=@n1, B=@n2, Alpha=@n4.RGB), B=Add(A=@n18, B=@n20))) {materialFunction=&MF_FX_EmissiveAdjustments}
 ]
}
