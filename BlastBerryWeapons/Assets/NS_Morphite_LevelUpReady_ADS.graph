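# Niagara level-up-ready effect for Morphite, apparently for aim-down-sights, showing short-lived blue streaks rising beside the camera.
# Only LevelUpLines001 is active, spawning thin vertical sprites in partition-selected boxes on either side of the camera, with an initial bright flash.
# Each update rebuilds positions from stored spawn offsets rotated to the camera vector, plus an age-based rise, keeping streaks view-locked.
# World-placed LevelUpLines and the empty ScreenEffectVignette emitter are disabled.
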NiagaraSystem /BlastBerryWeapons/Assets/NS_Morphite_LevelUpReady_ADS
dependencies {
 usedDataInterfaces [&NiagaraDataInterfaceCurve, &NiagaraDataInterfaceVector2DCurve, &NiagaraDataInterfaceCamera]
 usedDynamicInputs [
  &ReturnCameraPosition
  &NormalizeVector
  &Multiply_Vector
  &ReturnCameraVector
 ]
 usedModules [
  &SystemState
  &self:SystemSpawnScript.SystemScriptSource.SystemScriptGraph.NiagaraNodeAssignment_0.SetVariables_#1
 ]
}
emitters {
 LevelUpLines {
  bEnabled=false
  emitterUpdateScript [
   &EmitterState(Life Cycle Mode=::NewEnumerator1, Loop Behavior=::NewEnumerator1, Loop Duration=0.25, UseLoopDelay=false)
   &SpawnRate(SpawnRate=200, Use Spawn Probability=false)
  ]
  particleSpawnScript [
   &InitializeParticle(
    Color Maximum=<rgba 0.1195 0.4287 1 0.702>
    Color Minimum=<rgba 0.1195 0.4287 1 0.2>
    Color Mode=::NewEnumerator2
    Lifetime Max=0.5
    Lifetime Min=0.2
    Lifetime Mode=::NewEnumerator1
    Sprite Size Max=<2 80>
    Sprite Size Min=<1 10>
    Sprite Size Mode=::NewEnumerator2
    UsePositionOffset=false
    Write Material Random=true
   )
   &PartitionParticles()
   &ShapeLocation(
    Box Midpoint=<0.5 0.5 1>
    Box Size=<150 400 200>
    MaskSpawnGroup=false
    Offset=&MaskVector(Mask Source=1, Mask Parameter=$Particles.PartitionParticles.Partition, Mask=0, Masked Vector=<90 255 -85>, Passthrough Vector=<90 -230 -85>)
    Offset Coordinate Space=::Local
    Offset Mode=::NewEnumerator0
    Override Seed=false
    Rotation Coordinate Space=::World
    Shape Primitive=::NewEnumerator2
    Surface Only=false
    UseOwnerScale=true
   )
   &AddVelocity(
    Rotation Coordinate Space=::World
    Velocity=<0 0 1>
    Velocity Speed Scale=&RandomRangeFloat(Minimum=550, Maximum=800)
   )
  ]
  particleUpdateScript [
   &ParticleState()
   &ScaleSpriteSize(
    Non-Uniform Curve Sprite Scale=DataInterface() {propertyValues {
      XCurve {keys [{tangentMode=RCTM_User, time=0, value=2}, {tangentMode=RCTM_User, time=1, value=0}]}
      YCurve {keys [{tangentMode=RCTM_User, time=0, value=1}, {tangentMode=RCTM_User, time=1, value=0.65}]}
     }}
    Scale Sprite Size Mode=::NewEnumerator3
   )
   &Drag(Drag=2, Use Linear Drag=true, Use Rotational Drag=false)
   &SolveForcesAndVelocity(Clamp Velocity=false, Limit Acceleration=false)
   &ScaleColor(
    Scale Alpha=&FloatFromCurve(FloatCurve=DataInterface() {propertyValues {Curve {keys [<at 0 0>,<at 0.05 1>]}}})
    Scale RGB=&VectorFromFloat(Value=&FloatFromCurve(FloatCurve=DataInterface() {propertyValues {Curve {keys [<at 0 50>,<at 0.2 1>]}}}))
    ScaleA=true
    ScaleRGB=true
   )
   &SpriteFacingAndAlignment(Alignment Coordinate Space=::World, Sprite Alignment=<0 0 1>, WriteAlignment=true, WriteFacing=false)
  ]
  rendererClasses [&NiagaraSpriteRendererProperties]
  renderers [{rendererClass=&NiagaraSpriteRendererProperties}]
 }
 LevelUpLines001 {
  emitterUpdateScript [
   &EmitterState(Life Cycle Mode=::NewEnumerator1, Loop Behavior=::NewEnumerator1, Loop Duration=0.25, UseLoopDelay=false)
   &SpawnRate(SpawnRate=200, Use Spawn Probability=false)
  ]
  particleSpawnScript [
   &InitializeParticle(
    Color Maximum=<rgba 0.1195 0.4287 1 0.702>
    Color Minimum=<rgba 0.1195 0.4287 1 0.2>
    Color Mode=::NewEnumerator2
    Lifetime Max=0.55
    Lifetime Min=0.4
    Lifetime Mode=::NewEnumerator1
    Position=$System.CameraPos
    Position Mode=::NewEnumerator1
    Sprite Size Max=<0.2 5>
    Sprite Size Min=<0.1 0.5>
    Sprite Size Mode=::NewEnumerator2
    UsePositionOffset=false
    Write Material Random=true
   )
   &PartitionParticles()
   &ShapeLocation(
    Box Midpoint=<0.5 0.5 1>
    Box Size=&MaskVector(Mask Source=1, Mask Parameter=$Particles.PartitionParticles.Partition, Mask=0, Masked Vector=<22 8 15>, Passthrough Vector=<22 8 15>)
    MaskSpawnGroup=false
    Offset=&MaskVector(Mask Source=1, Mask Parameter=$Particles.PartitionParticles.Partition, Mask=0, Masked Vector=<10 -9 -15>, Passthrough Vector=<10 8 -15>)
    Offset Coordinate Space=::World
    Offset Mode=::NewEnumerator0
    Override Seed=false
    Rotation Coordinate Space=::World
    Shape Origin=$System.CameraPos
    Shape Primitive=::NewEnumerator2
    Surface Only=false
    UseOwnerScale=true
   )
   /SetVariables(Particles.InitialOffset:Vector3f=&SubtractVectorFromPosition(Position=$Particles.Position, Vector=Linked() {linkedVariable {name=System.CameraPos, type=NiagaraPosition}})) {isSetParameters=true}
   &SpriteFacingAndAlignment(Alignment Coordinate Space=::World, Sprite Alignment=<0 0 1>, WriteAlignment=true, WriteFacing=false)
  ]
  particleUpdateScript [
   &ParticleState()
   /SetVariables(Particles.Position=&AddVectorToPosition(
     Position=$System.CameraPos
     Vector=&Add_Vector(
      A=&RotateVectorByQuaternion(
       Vector To Rotate=$Particles.InitialOffset
       Quaternion=&MakeQuaternion(
        Quaternion From=3
        Axis=<1 0 0>
        Angle=0
        From=<1 0 0>
        To=$System.CameraVector
        Angle Type=::NewEnumerator0
        Coordinate Space=::World
        Normalize Vectors=true
       )
      )
      B=&Lerp_Vector(A=<0 0 0>, B=<0 0 35>, Alpha=$Particles.NormalizedAge)
     )
    )) {isSetParameters=true}
   &ScaleSpriteSize(
    Non-Uniform Curve Sprite Scale=DataInterface() {propertyValues {
      XCurve {keys [{tangentMode=RCTM_User, time=0, value=2}, {tangentMode=RCTM_User, time=1, value=0.5}]}
      YCurve {keys [{tangentMode=RCTM_User, time=0, value=1}, {tangentMode=RCTM_User, time=1, value=0.65}]}
     }}
    Scale Sprite Size Mode=::NewEnumerator3
   )
   &SolveForcesAndVelocity(Clamp Velocity=false, Limit Acceleration=false)
   &ScaleColor(
    Scale Alpha=&FloatFromCurve(FloatCurve=DataInterface() {propertyValues {Curve {keys [<at 0 0>,<at 0.05 1>,<at 0.8 1>,<at 1 0>]}}})
    Scale RGB=&VectorFromFloat(Value=&FloatFromCurve(FloatCurve=DataInterface() {propertyValues {Curve {keys [<at 0 50>,<at 0.2 1>]}}}))
    ScaleA=true
    ScaleRGB=true
   )
  ]
  rendererClasses [&NiagaraSpriteRendererProperties]
  renderers [{rendererClass=&NiagaraSpriteRendererProperties}]
 }
 ScreenEffectVignette {bEnabled=false, rendererClasses [], renderers []}
}
summary {
 systemName=NS_Morphite_LevelUpReady_ADS
 emitters [
  {bEnabled=false, emitterName=ScreenEffectVignette, rendererClasses []}
  {bEnabled=false, emitterName=LevelUpLines, rendererClasses [&NiagaraSpriteRendererProperties]}
  {emitterName=LevelUpLines001, rendererClasses [&NiagaraSpriteRendererProperties]}
 ]
 userVariables [User.Lifetime {type=NiagaraFloat, defaultValue=0.5, description=""}]
}
