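# Niagara system for a Morphite weapon effect, named as a level-up-ready cue, that flashes short-lived bursts of mesh rays.
# The Rays emitter bursts randomly sized, oriented and rotating meshes, with a loop-count mask setting full or reduced scale.
# Curves grow, dim and fade each ray and drive dynamic material parameters, while a distance LOD remap boosts growth for far views.
# FlareSpawn0 lists no renderers or modules.
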NiagaraSystem /BlastBerryWeapons/Assets/NS_Morphite_LevelUpReady_Weapon
dependencies {usedDataInterfaces [&NiagaraDataInterfaceCurve], usedDynamicInputs [], usedModules [&SystemState]}
emitters {
 FlareSpawn0 {rendererClasses [], renderers []}
 Rays {
  emitterUpdateScript [
   &EmitterState()
   &SpawnBurst_Instantaneous(
    Spawn Count=&RandomRangeInt(Minimum=8, Maximum=12)
    Use Spawn Probability=false
    UseLoopCountLimit=false
   )
  ]
  particleSpawnScript [
   /SetVariables(Particles.SpawnRotationVector=&RandomRangeVector(Minimum=<-0.5 0 0>, Maximum=<0.5 0 0>)) {isSetParameters=true}
   /SetVariables(Particles.SpawnRotationVector=&RandomRangeVector(Minimum=<-1 -1 -1>, Maximum=<1 1 1>)) {isSetParameters=true}
   &InitializeParticle(
    Lifetime=&RandomRangeFloat(Minimum=0.25, Maximum=0.4)
    Mesh Scale=&Multiply_VectorByFloat(
     Vector=/Niagara/DynamicInputs/CascadeConversion/SplitVector(
      Double Component Vector Channels:ECascadeNiagaraTwoVectorChannels=::NewEnumerator1
      Double Component:NiagaraFloat=&RandomRangeFloat(Minimum=0.4, Maximum=0.8)
      Single Component:NiagaraFloat=&RandomRangeFloat(Minimum=0.1, Maximum=0.25)
     )
     Float=&MaskFloat(Mask Source=1, Mask Parameter=$Emitter.LoopCount, Mask=0, Masked Float=1, Passthrough Float=0.65)
    )
    Mesh Scale Mode=::NewEnumerator4
    UsePositionOffset=false
    Write Material Random=true
   )
   &InitialMeshOrientation(
    Mesh Orientation Mode=::NewEnumerator2
    Orientation Coordinate Space=::Local
    Orientation Vector=&RandomRangeVector(Minimum=<0.1 -0.2 -0.2>, Maximum=<0.1 0.2 0.2>)
    Use Rotation Vector=false
   )
  ]
  particleUpdateScript [
   &ParticleState()
   &ScaleMeshSize(Scale Factor=&SplitVector(
     Double Component Vector Channels:ECascadeNiagaraTwoVectorChannels=::NewEnumerator1
     Double Component:NiagaraFloat=&FloatFromCurve(
      FloatCurve=DataInterface() {propertyValues {Curve {keys [{arriveTangent=3.0317, interpMode=RCIM_Cubic, time=0, value=0.0432}, <at 1 1.6255 RCIM_Cubic>]}}}
      Scale Curve=&FN_LOD_Distance_Remap(Start Distance=1000, End Distance=5000, Minimum Output=1, Maximum Output=3, Clamp Output=true)
     )
     Single Component:NiagaraFloat=&FloatFromCurve(
      FloatCurve=DataInterface() {propertyValues {Curve {keys [{arriveTangent=0.1226, interpMode=RCIM_Cubic, time=0, value=0.0866}, <at 1 1.9984 RCIM_Cubic>]}}}
      Scale Curve=&FN_LOD_Distance_Remap(Start Distance=1000, End Distance=5000, Minimum Output=1, Maximum Output=3, Clamp Output=true)
     )
    )) {module=ScaleMeshSize001}
   &UpdateMeshOrientation(Rotation Vector=$Particles.SpawnRotationVector)
   &Color(Color=&MakeLinearColorFromVectorAndFloat(
     "Vector (RGB)"=&VectorFromFloat(Value=&FloatFromCurve(FloatCurve=DataInterface() {propertyValues {Curve {keys [<at 0 0.9964 RCIM_Cubic>, {arriveTangent=-0.96, interpMode=RCIM_Cubic, leaveTangent=-0.96, time=0.5, value=0.293}, <at 1 0.0364>]}}}))
     "Float (Alpha)"=&ClampFloat(Float=&FloatFromCurve(FloatCurve=DataInterface() {propertyValues {Curve {keys [
           <at -0.002 0.3 RCIM_Cubic>
           {arriveTangent=-0.2994, interpMode=RCIM_Cubic, leaveTangent=-0.2994, time=0.1595, value=0.2}
           <at 1 0 RCIM_Cubic>
          ]}}}))
    ))
   &DynamicMaterialParameters(
    Index 0 Param 1=&FloatFromCurve(FloatCurve=DataInterface() {propertyValues {Curve {keys [{arriveTangent=-1.4398, interpMode=RCIM_Cubic, time=0, value=2.4783}, <at 1 1.5796 RCIM_Cubic>]}}})
    Index 0 Param 2=1
    Index 0 Param 3=&FloatFromCurve(FloatCurve=DataInterface() {propertyValues {Curve {keys [{arriveTangent=3.3568, interpMode=RCIM_Cubic, time=0, value=0.5548}, <at 1 1.5037>]}}})
    Index 0 Param 4=0.5
    Module.Index0Float0Write=true
    Module.Index0Float1Write=true
    Module.Index0Float2Write=true
    Module.Index0Float3Write=true
   )
  ]
  rendererClasses [&NiagaraMeshRendererProperties]
  renderers [{rendererClass=&NiagaraMeshRendererProperties, rendererIndex=0}]
 }
}
summary {
 systemName=NS_Morphite_LevelUpReady_Weapon
 emitters [{emitterName=Rays, rendererClasses [&NiagaraMeshRendererProperties]}, {emitterName=FlareSpawn0, rendererClasses []}]
 userVariables [User.Lifetime {type=NiagaraFloat, defaultValue=1, description=""}]
}
