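# Niagara effect for the Morphite weapon level-up reload activation, staging a timed transition along the weapon.
# Heavy_Cubes (meshes plus sprites, pulled inward) and curl-noise sparks (rate scaled by WeaponPhase) spawn on rings sliding one phase segment along X to LevelUpPos over TransitionTime.
# Rays bursts meshes at LevelUpPos when TransitionTime elapses.
# Distance-based remaps enlarge cube rings and rays for readability afar, while sparks distance-cull; FlareSpawn0 has no renderer.
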NiagaraSystem /BlastBerryWeapons/Assets/NS_Morphite_LevelUp_Reload_Activated
dependencies {
 usedDataInterfaces [&NiagaraDataInterfaceCurve]
 usedDynamicInputs [&MaskFloatBySpawnGroup, &Lerp_Float, &Add_Float]
 usedModules [
  &SystemState
  &self:SystemSpawnScript.SystemScriptSource.SystemScriptGraph.NiagaraNodeAssignment_0.SetVariables_#1
 ]
}
emitters {
 FlareSpawn0 {rendererClasses [], renderers []}
 Heavy_Cubes {
  emitterUpdateScript [
   &EmitterState(Life Cycle Mode=::NewEnumerator1, Loop Behavior=::NewEnumerator1, Loop Duration=$User.TransitionTime, UseLoopDelay=false)
   /SetVariables(
    Emitter.AnimatedOffset:Vector3f=&Lerp_Vector(
     A=&Add_Vector(
      A=Linked() {linkedVariable {name=System.LevelUpPos, type=NiagaraPosition}}
      B=&MakeVector(X=&Divide_Float(A=$User.WeaponLength, B=$User.WeaponPhaseAmount), Y=0, Z=0)
     )
     B=Linked() {linkedVariable {name=System.LevelUpPos, type=NiagaraPosition}}
     Alpha=$Emitter.NormalizedLoopAge
    )
    Emitter.EmissiveMult:NiagaraFloat=&FloatFromCurve(FloatCurve=DataInterface() {propertyValues {Curve {keys [<at 0 5 RCIM_Cubic>,<at 0.2 0>]}}}, CurveIndex=$Engine.System.Age)
   ) {isSetParameters=true}
   &SpawnRate(
    SpawnRate=&FloatFromCurve(FloatCurve=DataInterface() {propertyValues {Curve {keys [<at 0 1>,<at 1 0.5>]}}}, CurveIndex=$Emitter.NormalizedLoopAge, Scale Curve=350)
    Use Spawn Probability=false
   )
   &SpawnRate(
    SpawnGroup=1
    SpawnRate=&FloatFromCurve(FloatCurve=DataInterface() {propertyValues {Curve {keys [<at 0 1>,<at 1 0.5>]}}}, CurveIndex=$Emitter.NormalizedLoopAge, Scale Curve=50)
    Use Spawn Probability=false
   ) {module=SpawnRate001}
  ]
  particleSpawnScript [
   /SetVariables(Particles.VisibilityTag=&SpawnGroup()) {isSetParameters=true}
   &InitializeParticle(
    Color Maximum=<rgba 0.1023 0.2217 0.4844 1>
    Color Minimum=<rgba 0.0609 0.2235 0.33 1>
    Color Mode=::NewEnumerator2
    Lifetime Max=0.3
    Lifetime Min=0.1
    Lifetime Mode=::NewEnumerator1
    Mesh Scale Mode=::NewEnumerator1
    Mesh Uniform Scale Max=1
    Mesh Uniform Scale Min=0.55
    Sprite Size Mode=::NewEnumerator1
    Uniform Sprite Size Max=4
    Uniform Sprite Size Min=2
    UsePositionOffset=false
    Write Material Random=true
   )
   &DynamicMaterialParameters(
    Index 0 Param 1=&UniformRangedFloat(Minimum=0, Maximum=1)
    Index 0 Param 2=&UniformRangedFloat(Minimum=1, Maximum=2)
    Index 0 Param 3=1
    Module.Param0WriteEnabled=true
    Module.Param1WriteEnabled=true
    Module.Param2WriteEnabled=false
    Module.Param3WriteEnabled=false
   )
   &ShapeLocation(
    MaskSpawnGroup=false
    Offset=$Emitter.AnimatedOffset
    Offset Mode=::NewEnumerator0
    Override Seed=false
    Ring Radius=&FloatFromCurve(
     FloatCurve=DataInterface() {propertyValues {Curve {keys [<at 0 1>,<at 1 0.5>]}}}
     CurveIndex=$Emitter.NormalizedLoopAge
     Scale Curve=&Multiply_Float(
      A=&RandomRangeFloat(
       Minimum=25
       Maximum=&UniformAOrBFloat(A=30, B=40, Distribution Weight=0.35, Evaluation Type=::NewEnumerator0, Recalculate Random Each Loop=true, Override Seed=false)
      )
      B=&FN_LOD_Distance_Remap(Start Distance=500, End Distance=5000, Minimum Output=1, Maximum Output=2, Clamp Output=true)
     )
    )
    Rotation Mode=::NewEnumerator1
    Shape Primitive=::NewEnumerator4
    UseOwnerScale=true
   )
   &AddVelocity(
    Module.Use Falloff Distance=false
    Origin Offset=$Emitter.AnimatedOffset
    Override Seed=false
    Velocity Mode=::NewEnumerator1
    Velocity Speed=&Multiply_Float(
     A=&RandomRangeFloat(Minimum=-175, Maximum=-200)
     B=&FloatFromCurve(
      FloatCurve=DataInterface() {propertyValues {Curve {keys [{arriveTangent=-0.7989, interpMode=RCIM_Cubic, leaveTangent=-0.7989, tangentMode=RCTM_User, time=0, value=1}, <at 1 0.35>]}}}
      CurveIndex=$Emitter.NormalizedLoopAge
      Scale Curve=&FN_LOD_Distance_Remap(Start Distance=500, End Distance=5000, Minimum Output=1, Maximum Output=2, Clamp Output=true)
     )
    )
   )
   &InitialMeshOrientation(
    Mesh Orientation Mode=::NewEnumerator7
    Rotation=&RandomRangeVector(Minimum=<0 0 0>, Maximum=<0 0 0>)
    Use Rotation Vector=true
   )
  ]
  particleUpdateScript [
   &ParticleState()
   &ScaleColor(
    Scale Alpha=&FloatFromCurve(FloatCurve=DataInterface() {propertyValues {Curve {keys [<at 0.66 1 RCIM_Cubic>, {arriveTangent=-4.6302, interpMode=RCIM_Cubic, leaveTangent=-4.6302, tangentMode=RCTM_User, time=1, value=0}]}}})
    Scale RGB=&VectorFromFloat(Value=&FloatFromCurve(FloatCurve=DataInterface() {propertyValues {Curve {keys [
          {arriveTangent=-138.954, interpMode=RCIM_Cubic, leaveTangent=-138.954, tangentMode=RCTM_User, time=0, value=10}
          <at 0.1 1>
          {interpMode=RCIM_Cubic, leaveTangent=-29.5004, time=1, value=0}
         ]}}}))
    ScaleA=true
    ScaleRGB=true
   )
   /SetVariables(Particles.Color_Sprite:LinearColor=&MakeLinearColorFromVectorAndFloat(
     "Vector (RGB)"=&Multiply_VectorByFloat(
      Vector=&MakeVectorFromLinearColorRGB(Color=$Particles.Initial.Color)
      Float=&FloatFromCurve(FloatCurve=DataInterface() {propertyValues {Curve {keys [
           {arriveTangent=-138.954, interpMode=RCIM_Cubic, leaveTangent=-138.954, tangentMode=RCTM_User, time=0, value=10}
           <at 0.1 1.5>
           {interpMode=RCIM_Cubic, leaveTangent=-29.5004, time=1, value=0.5}
          ]}}})
     )
     "Float (Alpha)"=&FloatFromCurve(FloatCurve=DataInterface() {propertyValues {Curve {keys [
          {arriveTangent=9.0725, interpMode=RCIM_Cubic, leaveTangent=9.0725, tangentMode=RCTM_User, time=0, value=0}
          {interpMode=RCIM_Cubic, tangentMode=RCTM_User, time=0.1408, value=0.9847}
          {arriveTangent=-0.0405, interpMode=RCIM_Cubic, leaveTangent=-0.0405, tangentMode=RCTM_User, time=1, value=0}
         ]}}})
    )) {isSetParameters=true}
   &ScaleMeshSize(Scale Factor=&VectorFromFloat(Value=&FloatFromCurve(FloatCurve=DataInterface() {propertyValues {Curve {keys [
          {arriveTangent=13.966, interpMode=RCIM_Cubic, leaveTangent=13.966, tangentMode=RCTM_User, time=0, value=0.1}
          {arriveTangent=-0.4281, interpMode=RCIM_Cubic, leaveTangent=-0.4281, tangentMode=RCTM_User, time=0.1, value=1}
          {arriveTangent=-0.8279, tangentMode=RCTM_User, time=1, value=0.1}
         ]}}})))
   &ScaleSpriteSize(
    Scale Factor=&Lerp_Vector2(
     A=<1 1>
     B=<0.25 1.5>
     Alpha=&FloatFromCurve(FloatCurve=DataInterface() {propertyValues {Curve {keys [
          {arriveTangent=4.1785, interpMode=RCIM_Cubic, leaveTangent=4.1785, tangentMode=RCTM_User, time=0, value=0}
          {arriveTangent=1.1179, interpMode=RCIM_Cubic, leaveTangent=1.1179, tangentMode=RCTM_User, time=0.2042, value=0.6228}
          {arriveTangent=0.0224, interpMode=RCIM_Cubic, leaveTangent=0.0224, tangentMode=RCTM_User, time=1, value=1}
         ]}}})
    )
    Scale Sprite Size Mode=::NewEnumerator2
   )
   &UpdateMeshOrientation(Rotation Rate=&RandomRangeFloat(Minimum=0.75, Maximum=1.5), Rotation Vector=<0 0 1>)
   &Drag(
    Drag=&FloatFromCurve(
     FloatCurve=DataInterface() {propertyValues {Curve {keys [<at 0.5 1 RCIM_Cubic>, {arriveTangent=4.2548, tangentMode=RCTM_User, time=1, value=3}]}}}
     Scale Curve=3.5
    )
    Rotational Drag=0
    Use Linear Drag=true
    Use Rotational Drag=false
   )
   &SolveForcesAndVelocity(Clamp Velocity=false, Limit Acceleration=false)
   &CameraOffset(Camera Offset Amount=10)
  ]
  rendererClasses [&NiagaraMeshRendererProperties, &NiagaraSpriteRendererProperties]
  renderers [
   {rendererClass=&NiagaraMeshRendererProperties, rendererIndex=0}
   {rendererClass=&NiagaraSpriteRendererProperties, rendererIndex=1}
  ]
 }
 Rays {
  emitterUpdateScript [
   &EmitterState()
   &SpawnBurst_Instantaneous(Spawn Count=12, Spawn Time=$User.TransitionTime, Use Spawn Probability=false, UseLoopCountLimit=false)
  ]
  particleSpawnScript [
   /SetVariables(Particles.SpawnRotationVector=&RandomRangeVector(Minimum=<-0.5 0 0>, Maximum=<0.5 0 0>)) {isSetParameters=true}
   /SetVariables(Particles.SpawnRotationVector=&RandomRangeVector(Minimum=<-1 -1 -1>, Maximum=<1 1 1>)) {isSetParameters=true}
   &InitializeParticle(
    Lifetime=&RandomRangeFloat(Minimum=0.15, Maximum=0.25)
    Mesh Scale=/Niagara/DynamicInputs/CascadeConversion/SplitVector(
     Double Component Vector Channels:ECascadeNiagaraTwoVectorChannels=::NewEnumerator1
     Double Component:NiagaraFloat=&RandomRangeFloat(Minimum=0.25, Maximum=0.4)
     Single Component:NiagaraFloat=&RandomRangeFloat(Minimum=0.135, Maximum=0.2)
    )
    Mesh Scale Mode=::NewEnumerator4
    UsePositionOffset=false
    Write Material Random=true
   )
   &SystemLocation(MaskSpawnGroup=false, Offset:NiagaraPosition=$System.LevelUpPos)
   &InitialMeshOrientation(
    Mesh Orientation Mode=::NewEnumerator2
    Orientation Vector=&RandomRangeVector(Minimum=<1 -0.2 -0.2>, Maximum=<1 0.2 0.2>)
    Use Rotation Vector=false
   )
  ]
  particleUpdateScript [
   &ParticleState()
   &ScaleMeshSize(Scale Factor=&SplitVector(
     Double Component Vector Channels:ECascadeNiagaraTwoVectorChannels=::NewEnumerator1
     Double Component:NiagaraFloat=&FloatFromCurve(
      FloatCurve=DataInterface() {propertyValues {Curve {keys [{arriveTangent=3.0317, interpMode=RCIM_Cubic, time=0, value=0.0432}, <at 1 1.6255 RCIM_Cubic>]}}}
      Scale Curve=&FN_LOD_Distance_Remap(Start Distance=1000, End Distance=5000, Minimum Output=1, Maximum Output=3, Clamp Output=true)
     )
     Single Component:NiagaraFloat=&FloatFromCurve(
      FloatCurve=DataInterface() {propertyValues {Curve {keys [{arriveTangent=0.1226, interpMode=RCIM_Cubic, time=0, value=0.0866}, <at 1 1.9984 RCIM_Cubic>]}}}
      Scale Curve=&FN_LOD_Distance_Remap(Start Distance=1000, End Distance=5000, Minimum Output=1, Maximum Output=3, Clamp Output=true)
     )
    )) {module=ScaleMeshSize001}
   &UpdateMeshOrientation(Rotation Vector=$Particles.SpawnRotationVector)
   &Color(Color=&MakeLinearColorFromVectorAndFloat(
     "Vector (RGB)"=&VectorFromFloat(Value=&FloatFromCurve(FloatCurve=DataInterface() {propertyValues {Curve {keys [<at 0 0.9964 RCIM_Cubic>, {arriveTangent=-0.96, interpMode=RCIM_Cubic, leaveTangent=-0.96, time=0.5, value=0.293}, <at 1 0.0364>]}}}))
     "Float (Alpha)"=&ClampFloat(Float=&FloatFromCurve(FloatCurve=DataInterface() {propertyValues {Curve {keys [
           <at -0.002 0.0024 RCIM_Cubic>
           {arriveTangent=-0.0024, interpMode=RCIM_Cubic, leaveTangent=-0.0024, time=0.1595, value=0.2}
           <at 1 0 RCIM_Cubic>
          ]}}}))
    ))
   &DynamicMaterialParameters(
    Index 0 Param 1=&FloatFromCurve(FloatCurve=DataInterface() {propertyValues {Curve {keys [{arriveTangent=-1.4398, interpMode=RCIM_Cubic, time=0, value=2.4783}, <at 1 1.5796 RCIM_Cubic>]}}})
    Index 0 Param 2=1
    Index 0 Param 3=&FloatFromCurve(FloatCurve=DataInterface() {propertyValues {Curve {keys [{arriveTangent=3.3568, interpMode=RCIM_Cubic, time=0, value=0.5548}, <at 1 1.5037>]}}})
    Index 0 Param 4=0.5
    Module.Index0Float0Write=true
    Module.Index0Float1Write=true
    Module.Index0Float2Write=true
    Module.Index0Float3Write=true
   )
  ]
  rendererClasses [&NiagaraMeshRendererProperties]
  renderers [{rendererClass=&NiagaraMeshRendererProperties, rendererIndex=0}]
 }
 Spark001 {
  emitterUpdateScript [
   &EmitterState(
    Enable Distance Culling=true
    Life Cycle Mode=::NewEnumerator1
    Loop Behavior=::NewEnumerator1
    Loop Delay=0.05
    Loop Duration=$User.TransitionTime
    MaxDistance=3500
    Module.bUseMaxDistance=true
    Module.bUseMinDistance=false
    Scalability Mode=::NewEnumerator1
    UseLoopDelay=false
   )
   &SpawnRate(
    SpawnRate=&Multiply_Float(A=&RandomRangeFloat(Minimum=25, Maximum=50), B=$User.WeaponPhase)
    Use Spawn Probability=false
   )
   /SetVariables(Emitter.AnimatedOffset:Vector3f=&Lerp_Vector(
     A=&Add_Vector(
      A=Linked() {linkedVariable {name=System.LevelUpPos, type=NiagaraPosition}}
      B=&MakeVector(X=&Divide_Float(A=$User.WeaponLength, B=$User.WeaponPhaseAmount), Y=0, Z=0)
     )
     B=Linked() {linkedVariable {name=System.LevelUpPos, type=NiagaraPosition}}
     Alpha=$Emitter.NormalizedLoopAge
    )) {isSetParameters=true}
  ]
  particleSpawnScript [
   &InitializeParticle(
    Color Maximum=<rgba 0.1022 0.6358 1 10>
    Color Minimum=<rgba 0.1856 0.167 1 10>
    Color Mode=::NewEnumerator2
    Lifetime Max=0.4
    Lifetime Min=0.2
    Lifetime Mode=::NewEnumerator1
    Sprite Size Max=<1.5 2.5>
    Sprite Size Min=<0.75 1.25>
    Sprite Size Mode=::NewEnumerator2
    UsePositionOffset=false
    Write Material Random=true
   )
   &ShapeLocation(
    MaskSpawnGroup=false
    Offset=$Emitter.AnimatedOffset
    Offset Mode=::NewEnumerator0
    Override Seed=false
    Ring Radius=&RandomRangeFloat(Minimum=5, Maximum=10)
    Rotation Mode=::NewEnumerator1
    Shape Primitive=::NewEnumerator4
    UseOwnerScale=true
   )
   &AddVelocity(
    Module.Use Falloff Distance=false
    Origin Offset:NiagaraPosition=$System.LevelUpPos
    Override Seed=false
    Velocity Mode=::NewEnumerator1
    Velocity Origin=<0 0 0>
    Velocity Speed=&RandomRangeFloat(Minimum=75, Maximum=350)
   )
   &CameraOffset(Camera Offset Amount=&Multiply_Float(A=30, B=$System.CameraOffsetScale))
  ]
  particleUpdateScript [
   &ParticleState()
   &ScaleColor(
    Scale RGB=&VectorFromFloat(Value=&FloatFromCurve(FloatCurve=DataInterface() {propertyValues {Curve {keys [<at 0 10>,<at 0.5 1>]}}}))
    ScaleA=false
    ScaleRGB=true
   ) {module=ScaleColor001}
   &Drag(Drag=&RandomRangeFloat(Minimum=11, Maximum=20), Use Linear Drag=true, Use Rotational Drag=false)
   &CurlNoiseForce(Noise Strength=&RandomRangeFloat(Minimum=200, Maximum=400), Pan Noise=true)
   &GravityForce(Gravity=&RandomRangeVector(Minimum=<0 0 0>, Maximum=<0 0 -100>))
   &SolveForcesAndVelocity(Clamp Velocity=false, Limit Acceleration=false)
   &ScaleSpriteSize(Scale Factor=&MakeVector2D(
     X=&FloatFromCurve(FloatCurve=DataInterface() {propertyValues {Curve {keys [<at 0 0.5>,<at 0.1 1>,<at 1 0.5>]}}})
     Y=&FloatFromCurve(FloatCurve=DataInterface() {propertyValues {Curve {keys [<at 0 1>,<at 0.5 0.2>]}}}, Scale Curve=2)
    ))
   &FN_Sprite_Attributes(
    Color=&MakeLinearColorFromVectorAndFloat(
     "Vector (RGB)"=&MakeVectorFromLinearColorRGB(Color=&RandomRangeLinearColor(Minimum=<rgba 0.2281 0.3283 1 10>, Maximum=<rgba 0.1534 0.7041 1 10>))
     "Float (Alpha)"=&FloatFromCurve(FloatCurve=DataInterface() {propertyValues {Curve {keys [<at 0 1>,<at 0.8 0>]}}}, Scale Curve=0.5)
    )
    Create Color Attribute=true
    Create SpriteSize Attribute=true
    SpriteSize=&Multiply_Vector2D_ByFloat(
     Float=2
     Vector2D=&Multiply_Vector2D(
      A=<5 15>
      B=&MakeVector2D(
       X=&FloatFromCurve(FloatCurve=DataInterface() {propertyValues {Curve {keys [<at 0 0.5>,<at 0.1 1>,<at 1 0.5>]}}})
       Y=&FloatFromCurve(FloatCurve=DataInterface() {propertyValues {Curve {keys [<at 0 1>,<at 1 0>]}}})
      )
     )
    )
    Write to Module Namespace=true
   )
  ]
  rendererClasses [&NiagaraSpriteRendererProperties]
  renderers [
   {rendererClass=&NiagaraSpriteRendererProperties}
   {rendererClass=&NiagaraSpriteRendererProperties, rendererIndex=1}
  ]
 }
}
summary {
 systemName=NS_Morphite_LevelUp_Reload_Activated
 emitters [
  {emitterName=Spark001, rendererClasses [&NiagaraSpriteRendererProperties]}
  {emitterName=Rays, rendererClasses [&NiagaraMeshRendererProperties]}
  {emitterName=FlareSpawn0, rendererClasses []}
  {emitterName=Heavy_Cubes, rendererClasses [&NiagaraMeshRendererProperties, &NiagaraSpriteRendererProperties]}
 ]
 userVariables [
  User.TransitionTime {type=NiagaraFloat, defaultValue=0.5, description=""}
  User.WeaponLength {type=NiagaraFloat, defaultValue=95, description=""}
  User.WeaponPhase {type=NiagaraFloat, defaultValue=1, description=""}
  User.WeaponPhaseAmount {type=NiagaraFloat, defaultValue=4, description=""}
 ]
}
