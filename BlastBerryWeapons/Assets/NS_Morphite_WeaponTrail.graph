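# Niagara weapon trail effect for the Morphite weapon: a slow, sparse stream of blue-tinted particles drifting along negative X.
# A single Heavy_Cubes emitter runs two spawn rates, writing each spawn group to VisibilityTag, and carries both mesh and sprite renderers.
# Particles spawn in an owner-scaled volume with randomized material parameters, then slow under light drag.
# Curves flash them bright, fade them and animate mesh and sprite size, with sprites using a separate color attribute.
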NiagaraSystem /BlastBerryWeapons/Assets/NS_Morphite_WeaponTrail
dependencies {
 usedDataInterfaces [&NiagaraDataInterfaceCurve]
 usedDynamicInputs [&MaskFloatBySpawnGroup, &Lerp_Float]
 usedModules [&SystemState]
}
emitters {Heavy_Cubes {
  emitterUpdateScript [
   &EmitterState(Life Cycle Mode=::NewEnumerator1, Loop Duration=1, UseLoopDelay=false)
   &SpawnRate(SpawnRate=3.5, Use Spawn Probability=false)
   &SpawnRate(SpawnGroup=1, SpawnRate=1.5, Use Spawn Probability=false) {module=SpawnRate001}
  ]
  particleSpawnScript [
   /SetVariables(Particles.VisibilityTag=&SpawnGroup()) {isSetParameters=true}
   &InitializeParticle(
    Color Maximum=<rgba 0.1023 0.2217 0.4844 1>
    Color Minimum=<rgba 0.0609 0.2235 0.33 1>
    Color Mode=::NewEnumerator2
    Lifetime Max=4
    Lifetime Min=3
    Lifetime Mode=::NewEnumerator1
    Mesh Scale Mode=::NewEnumerator1
    Mesh Uniform Scale Max=0.75
    Mesh Uniform Scale Min=0.55
    Sprite Size Mode=::NewEnumerator1
    Uniform Sprite Size Max=4
    Uniform Sprite Size Min=2
    UsePositionOffset=false
    Write Material Random=true
   )
   &DynamicMaterialParameters(
    Index 0 Param 1=&UniformRangedFloat(Minimum=0, Maximum=1)
    Index 0 Param 2=&UniformRangedFloat(Minimum=1, Maximum=2)
    Index 0 Param 3=1
    Module.Param0WriteEnabled=true
    Module.Param1WriteEnabled=true
    Module.Param2WriteEnabled=false
    Module.Param3WriteEnabled=false
   )
   &ShapeLocation(
    Box Size=<20 12 15>
    MaskSpawnGroup=false
    Offset Mode=::NewEnumerator0
    Override Seed=false
    Rotation Mode=::NewEnumerator1
    Shape Primitive=::NewEnumerator2
    Surface Only=false
    UseOwnerScale=true
   )
   &AddVelocity(Velocity=<-1 0 0>, Velocity Speed Scale=&RandomRangeFloat(Minimum=8, Maximum=15))
   &InitialMeshOrientation(
    Mesh Orientation Mode=::NewEnumerator7
    Rotation=&RandomRangeVector(Minimum=<0 0 0>, Maximum=<0 0 0>)
    Use Rotation Vector=true
   )
  ]
  particleUpdateScript [
   &ParticleState()
   &ScaleColor(
    Scale Alpha=&FloatFromCurve(FloatCurve=DataInterface() {propertyValues {Curve {keys [<at 0.66 1 RCIM_Cubic>, {arriveTangent=-4.6302, interpMode=RCIM_Cubic, leaveTangent=-4.6302, tangentMode=RCTM_User, time=1, value=0}]}}})
    Scale RGB=&VectorFromFloat(Value=&FloatFromCurve(FloatCurve=DataInterface() {propertyValues {Curve {keys [
          {arriveTangent=-138.954, interpMode=RCIM_Cubic, leaveTangent=-138.954, tangentMode=RCTM_User, time=0, value=10}
          <at 0.1 1>
          {interpMode=RCIM_Cubic, leaveTangent=-29.5004, time=1, value=0}
         ]}}}))
    ScaleA=true
    ScaleRGB=true
   )
   /SetVariables(Particles.Color_Sprite:LinearColor=&MakeLinearColorFromVectorAndFloat(
     "Vector (RGB)"=&Multiply_VectorByFloat(
      Vector=&MakeVectorFromLinearColorRGB(Color=$Particles.Initial.Color)
      Float=&FloatFromCurve(FloatCurve=DataInterface() {propertyValues {Curve {keys [
           {
            arriveTangent=-62.465
            interpMode=RCIM_Cubic
            leaveTangent=-62.465
            leaveTangentWeight=2.0824
            tangentMode=RCTM_User
            tangentWeightMode=RCTWM_WeightedBoth
            time=0
            value=5
           }
           <at 0.1 1.5>
           {interpMode=RCIM_Cubic, leaveTangent=-29.5004, time=1, value=0.5}
          ]}}})
     )
     "Float (Alpha)"=&FloatFromCurve(FloatCurve=DataInterface() {propertyValues {Curve {keys [
          {arriveTangent=9.0725, interpMode=RCIM_Cubic, leaveTangent=9.0725, tangentMode=RCTM_User, time=0, value=0}
          {interpMode=RCIM_Cubic, tangentMode=RCTM_User, time=0.2, value=0.9847}
          {arriveTangent=-0.0405, interpMode=RCIM_Cubic, leaveTangent=-0.0405, tangentMode=RCTM_User, time=1, value=0}
         ]}}})
    )) {isSetParameters=true}
   &ScaleMeshSize(Scale Factor=&VectorFromFloat(Value=&FloatFromCurve(FloatCurve=DataInterface() {propertyValues {Curve {keys [
          {arriveTangent=7.8713, interpMode=RCIM_Cubic, leaveTangent=7.8713, tangentMode=RCTM_User, time=0, value=0.1}
          {arriveTangent=-0.4281, interpMode=RCIM_Cubic, leaveTangent=-0.4281, tangentMode=RCTM_User, time=0.25, value=1}
          {arriveTangent=-0.8279, tangentMode=RCTM_User, time=1, value=0.1}
         ]}}})))
   &ScaleSpriteSize(
    Scale Factor=&Lerp_Vector2(
     A=<0.5 1>
     B=<0.25 1.5>
     Alpha=&FloatFromCurve(FloatCurve=DataInterface() {propertyValues {Curve {keys [
          {arriveTangent=3.4162, interpMode=RCIM_Cubic, leaveTangent=3.4162, tangentMode=RCTM_User, time=0, value=0}
          {arriveTangent=1.1179, interpMode=RCIM_Cubic, leaveTangent=1.1179, tangentMode=RCTM_User, time=0.3001, value=0.584}
          {arriveTangent=0.0224, interpMode=RCIM_Cubic, leaveTangent=0.0224, tangentMode=RCTM_User, time=1, value=1}
         ]}}})
    )
    Scale Sprite Size Mode=::NewEnumerator2
   )
   &Drag(Drag=0.1, Rotational Drag=0, Use Linear Drag=true, Use Rotational Drag=false)
   &SolveForcesAndVelocity(Clamp Velocity=false, Limit Acceleration=false)
  ]
  rendererClasses [&NiagaraMeshRendererProperties, &NiagaraSpriteRendererProperties]
  renderers [
   {rendererClass=&NiagaraMeshRendererProperties, rendererIndex=0}
   {rendererClass=&NiagaraSpriteRendererProperties, rendererIndex=1}
  ]
 }}
summary {
 systemName=NS_Morphite_WeaponTrail
 emitters [{emitterName=Heavy_Cubes, rendererClasses [&NiagaraMeshRendererProperties, &NiagaraSpriteRendererProperties]}]
 userVariables []
}
